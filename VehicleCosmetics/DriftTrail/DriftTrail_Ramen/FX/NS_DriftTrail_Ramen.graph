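# Niagara system for the Ramen drift trail vehicle cosmetic, spawning noodle, soup and droplet effects at the wheels.
# Ribbons compute four sine-offset position attributes for wavy noodle strands, alongside a GPU soup-cone emitter, spinning sub-UV sprites, drops and scaled wheel-wrap meshes.
# Emitters follow the vehicle skeletal mesh and mask particles per tire through FN_DisableTireFX.
# Paint-variant overrides set most colors, while ColorVariantID selects soup material values.
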NiagaraSystem /VehicleCosmetics/DriftTrail/DriftTrail_Ramen/FX/NS_DriftTrail_Ramen
dependencies {
 usedDataInterfaces [
  &NiagaraDataInterfaceVectorCurve
  &NiagaraDataInterfaceCurve
  &NiagaraDataInterfaceSpriteRendererInfo
  &NiagaraDataInterfaceSkeletalMesh
 ]
 usedDynamicInputs []
 usedModules [
  &self:SystemSpawnScript.SystemScriptSource.SystemScriptGraph.NiagaraNodeAssignment_0.SetVariables_#1
  &FN_UnpackVehicleFlags
  /VehicleCosmetics/FX/Niagara/Modules/VC_SetVehicleSpeedAttribs
  &SystemState
 ]
}
emitters {
 MSH_Cone_Soup {
  simTarget=GPUComputeSim
  emitterSpawnScript [
   /SetVariables(Emitter.SUBSURFACEEXPPOWER:NiagaraFloat=&SelectFloatFromArray(
     Direct Array Index=$User.ColorVariantID
     Float Selection Array=DataInterface() {propertyValues {FloatData [-3,-0.5,-0.5,-0.5,1,0,-0.5,-0.5,0,20,-1.5,-1,50,-0.2,0]}}
    )) {isSetParameters=true}
   /SetVariables(Emitter.Roughness:NiagaraFloat=&SelectFloatFromArray(
     Direct Array Index=$User.ColorVariantID
     Float Selection Array=DataInterface() {propertyValues {FloatData [10,2,2,2,3,0.5,2,2,2,10,1,2,10,0,-20]}}
    )) {isSetParameters=true}
   /SetVariables(Emitter.Metalness:NiagaraFloat=&SelectFloatFromArray(
     Direct Array Index=$User.ColorVariantID
     Float Selection Array=DataInterface() {propertyValues {FloatData [2,1,0,0,0,1,-1,0,0.5,0.3,0,0,0,0,0]}}
    )) {isSetParameters=true}
  ]
  emitterUpdateScript [
   &EmitterState(
    Inactive Response=::NewEnumerator1
    Life Cycle Mode=::NewEnumerator1
    Loop Duration=0.3
    Recalculate Duration Each Loop=true
    UseLoopDelay=false
   )
   &SpawnRate(SpawnRate=15, Use Spawn Probability=false)
  ]
  particleSpawnScript [
   &VC_PaintedVariantsOverride(
    Black=<rgba 0.015 0.015 0.015 1>
    BurntSienna=<rgba 0.12 0.03 0.012 1>
    Cobalt=<rgba 0 0.0148 0.8 1>
    Crimson=<rgba 0.6 0.024 0.0432 1>
    DefaultColor=<rgba 0.85 0.4888 0.1275 1>
    ForestGreen=<rgba 0.011 0.22 0.011 1>
    Grey=<rgba 0.12 0.12 0.12 1>
    Lime=<rgba 0.2 0.4 0 1>
    Orange=<rgba 0.7 0.1167 0 1>
    Pink=<rgba 0.6 0.12 0.456 1>
    Purple=<rgba 0.1514 0.019 0.38 1>
    Saffron=<rgba 0.4 0.3667 0 1>
    SkyBlue=<rgba 0 0.125 0.5 1>
    TitaniumWhite=<rgba 0.45 0.45 0.45 1>
   )
   &InitializeParticle(
    Color Mode=::NewEnumerator0
    Lifetime=0.8
    Mesh Scale=<0.6 0.6 1.25>
    Mesh Scale Mode=::NewEnumerator4
    Sprite Rotation Mode=::NewEnumerator2
    UsePositionOffset=false
    Write Material Random=true
   )
   &PartitionParticles()
   &SkeletalMeshLocation(
    Bone / Socket Index=&Add_Integer(A=$Particles.PartitionParticles.Partition, B=4)
    Bone / Socket Sampling Mode=::NewEnumerator5
    Interpolate Mesh Sampling=false
    MaskSpawnGroup=false
    Mesh Sampling Type=::NewEnumerator6
    Sampled Position Offset=&AddVectorToPosition(Position=<-45 0 -20>, Vector=$User.offset)
    Skeletal Mesh=$System.SkeletalMesh
   )
   &InitialMeshOrientation(
    Mesh Axis To Orient=<0 0 0>
    Mesh Orientation Mode=::NewEnumerator2
    Orientation Vector=<0 0 0>
    Rotation=&RandomRangeVector(Minimum=<1 0 0>, Maximum=<-1 0 0>)
    Rotation Coordinate Space=::NewEnumerator0
    Use Rotation Vector=true
   )
   &AddVelocity(
    Velocity=/VehicleCosmetics/FX/Niagara/Modules/MergeEmitters/VC_VectorBySpawnGroup(
     SpawnGroup or ExecIndex:NiagaraInt32=0
     CustomInt:NiagaraInt32=0
     MaximumSpawnGroups:NiagaraInt32=2
     Vector Group 0:Vector3f=<80 0 0>
     Vector Group 1:Vector3f=<3000 0 0>
     Vector Group 2:Vector3f=<0 0 0>
     Vector Group 3:Vector3f=<0 0 0>
     Vector Group 4:Vector3f=<0 0 0>
     Vector Group 5:Vector3f=<0 0 0>
    )
    Velocity Speed Scale=-1
   ) {enabled=false}
  ]
  particleUpdateScript [
   &FN_DisableTireFX(
    Mask Index Attribute:NiagaraInt32=$Particles.PartitionParticles.Partition
    Mask Index FL=2
    Mask Index FR=3
    Mask Index RL=0
    Mask Index RR=1
   )
   &ParticleState()
   &ScaleMeshSize(Scale Factor=&VectorFromCurve(
     VectorCurve=DataInterface() {propertyValues {
       XCurve {keys [
         {arriveTangent=0.2699, interpMode=RCIM_Cubic, leaveTangent=0.2699, leaveTangentWeight=0.3333, tangentMode=RCTM_User, time=0, value=0.5}
         {arriveTangent=2.9664, arriveTangentWeight=0.3333, interpMode=RCIM_Cubic, leaveTangent=2.9664, tangentMode=RCTM_User, time=1, value=2.5}
        ]}
       YCurve {keys [
         {arriveTangent=0.2699, interpMode=RCIM_Cubic, leaveTangent=0.2699, leaveTangentWeight=0.3333, tangentMode=RCTM_User, time=0, value=0.5}
         {arriveTangent=2.9664, arriveTangentWeight=0.3333, interpMode=RCIM_Cubic, leaveTangent=2.9664, tangentMode=RCTM_User, time=1, value=2.5}
        ]}
       ZCurve {keys [
         {arriveTangent=0.0111, interpMode=RCIM_Cubic, leaveTangent=0.0111, tangentMode=RCTM_User, time=0, value=1}
         {arriveTangent=2.4313, interpMode=RCIM_Cubic, leaveTangent=2.4313, tangentMode=RCTM_User, time=0.9992, value=4.5}
        ]}
      }}
     Scale Curve=<1.6 1.6 0.9>
    ))
   &SolveForcesAndVelocity(Clamp Velocity=false, Limit Acceleration=false)
   &ScaleColor(
    Scale Alpha=&FloatFromCurve(
     FloatCurve=DataInterface() {propertyValues {Curve {
        defaultValue=1
        keys [
         {arriveTangent=43.5199, interpMode=RCIM_Cubic, leaveTangent=43.5199, tangentMode=RCTM_User, time=0, value=1}
         {interpMode=RCIM_Cubic, tangentMode=RCTM_User, time=0.2, value=10}
         {arriveTangent=-11.25, interpMode=RCIM_Cubic, leaveTangent=-11.25, time=0.4, value=10}
         {arriveTangentWeight=0.2667, interpMode=RCIM_Cubic, leaveTangent=-1.7044, time=1, value=1}
        ]
       }}}
     Scale Curve=&SelectFloatFromArray(
      Array Sampling Mode=::NewEnumerator0
      Direct Array Index=?
      Float Selection Array=DataInterface() {propertyValues {FloatData [0.2,0.25]}}
      Random Seed=0
     )
    )
    Scale RGB=&VectorFromFloat(Value=&FloatFromCurve(
      FloatCurve=DataInterface() {propertyValues {Curve {
         defaultValue=1
         keys [{interpMode=RCIM_Cubic, tangentMode=RCTM_User, time=0, value=1}, {interpMode=RCIM_Cubic, tangentMode=RCTM_User, time=0.5, value=0.5}]
        }}}
      Scale Curve=&MakeCustomFloatFromBool(Bool=$User.bIsPainted, True Float=1.25, False Float=1.25)
     ))
    ScaleA=true
    ScaleRGB=true
   )
  ]
  rendererClasses [&NiagaraMeshRendererProperties, &NiagaraSpriteRendererProperties]
  renderers [
   {rendererClass=&NiagaraMeshRendererProperties, rendererIndex=0}
   {rendererClass=&NiagaraSpriteRendererProperties, rendererIndex=1}
  ]
 }
 NE_VC_DriftTrail_Ribbon_Parent001 {
  emitterUpdateScript [
   &EmitterState(Inactive Response=::NewEnumerator1, Life Cycle Mode=::NewEnumerator1, Loop Duration=1, UseLoopDelay=false)
   &VC_RateQualityHandler(High Spawn Rate=50, Low Spawn Rate=30, Medium Spawn Rate=40, Use Custom Rate?=true)
   &VC_SpawnRibbon_v1(Number of Ribbons=2, SpawnRate=50)
  ]
  particleSpawnScript [
   &VC_PaintedVariantsOverride(
    BurntSienna=<rgba 0.08 0.0247 0.0099 0.8435>
    Cobalt=<rgba 0.0136 0.0112 0.6042 1>
    Crimson=<rgba 0.2552 0.0265 0.0367 1>
    DefaultColor=<rgba 0.8 0.3658 0.12 1>
    ForestGreen=<rgba 0.0089 0.2135 0.0192 1>
    Grey=<rgba 0.0677 0.0677 0.0677 1>
    Lime=<rgba 0.1224 0.2448 0 1>
    Orange=<rgba 0.2552 0.061 0.0106 1>
    Pink=<rgba 0.3281 0.0684 0.2337 1>
    Purple=<rgba 0.0953 0.024 0.2 1>
    Saffron=<rgba 0.3009 0.2609 0.0063 1>
    SkyBlue=<rgba 0 0.1492 0.4844 1>
   )
   &VC_InitializeRibbon_v1(
    Color=$Particles.Color
    Custom Offset=true
    Enable Skeletal Position=true
    Facing Axis:ENiagaraOrientationAxis=::YAxis
    Lifetime=0.4
    Local Vector Offset=&AddVectorToPosition(Position=<-55 0 -20>, Vector=$User.offset)
    Offset Amount=1
    Ribbon Width=4
   )
   &AddVelocity(Velocity=&MakeVector(X=-2000, Y=0, Z=0)) {module=AddVelocity001}
  ]
  particleUpdateScript [
   &FN_DisableTireFX(Mask Index Attribute:NiagaraInt32=$Particles.SpawnGroup, Mask Index FL=2, Mask Index FR=3, Mask Index RL=0, Mask Index RR=1)
   /SetVariables(Particles.RibbonFacing=&MakeVector(
     X=0
     Y=&Sine(
      Normalized Angle=&Subtract_Float(A=$Emitter.NormalizedLoopAge, B=$Particles.NormalizedAge)
      Period=1
      Scale=&FloatFromCurve(FloatCurve=DataInterface() {propertyValues {Curve {
          defaultValue=1
          keys [
           {arriveTangent=4.1785, interpMode=RCIM_Cubic, leaveTangent=4.1785, tangentMode=RCTM_User, time=0, value=0}
           {arriveTangent=0.0224, interpMode=RCIM_Cubic, leaveTangent=0.0224, tangentMode=RCTM_User, time=0.5, value=1}
          ]
         }}})
      Bias=0
     )
     Z=1
    )) {enabled=false, isSetParameters=true}
   &ParticleState()
   &InheritSourceMovement(
    Applied Position Scale=&Multiply_VectorByFloat(
     Vector=<1 1 1>
     Float=&FloatFromCurve(FloatCurve=DataInterface() {propertyValues {Curve {keys [
          {interpMode=RCIM_Cubic, leaveTangent=-0.1311, tangentMode=RCTM_Break, time=0, value=1}
          {arriveTangent=-0.0803, interpMode=RCIM_Cubic, tangentMode=RCTM_Break, time=1, value=0.9}
         ]}}})
    )
    ApplyForce=false
    ApplyPosition=true
    ApplyPositionFalloff=false
   )
   &SolveForcesAndVelocity(Clamp Velocity=false, Limit Acceleration=false)
   &ScaleColor(Scale RGB=&VectorFromFloat(Value=1.9), ScaleA=true, ScaleRGB=true)
   &VC_UpdateRibbon_v1(Enable Skeletal Position=true)
   &ScaleRibbonWidth(Ribbon Width Scale=&FloatFromCurve(FloatCurve=DataInterface() {propertyValues {Curve {
        defaultValue=1
        keys [
         {tangentMode=RCTM_User, time=0, value=0}
         <at 0.05 1>
         {arriveTangent=-0.0374, interpMode=RCIM_Cubic, leaveTangent=-0.0374, tangentMode=RCTM_User, time=0.85, value=1}
         {arriveTangent=-12.8345, interpMode=RCIM_Cubic, leaveTangent=-12.8345, tangentMode=RCTM_User, time=0.89, value=0.6571}
         {arriveTangent=-68.1629, interpMode=RCIM_Cubic, leaveTangent=-68.1629, tangentMode=RCTM_User, time=0.9, value=0.4}
        ]
       }}}))
   /SetVariables(
    Particles.PosNoodle1:NiagaraPosition=&AddVectorToPosition(
     Position=$Particles.Position
     Vector=&TransformVector(
      Vector=&MakeVector(
       X=-10
       Y=&Sine(
        Normalized Angle=&Add_Float(
         A=&Multiply_Float(A=$Particles.NormalizedAge, B=-1.3)
         B=&Multiply_Float(A=$Emitter.LoopedAge, B=1)
        )
        Period=&MaskFloatBySpawnGroup(Masked Float=0.6, Passthrough Float=-0.6)
        Scale=&FloatFromCurve(
         FloatCurve=DataInterface() {propertyValues {Curve {keys [{arriveTangent=0.1945, interpMode=RCIM_Cubic, leaveTangent=0.1945, tangentMode=RCTM_User, time=0, value=0}, <at 0.2717 1>, <at 1 5>]}}}
         Scale Curve=4
        )
        Bias=0
       )
       Z=&Sine(
        Normalized Angle=&Add_Float(
         A=&Multiply_Float(A=$Particles.NormalizedAge, B=-1.3)
         B=&Multiply_Float(A=$Emitter.LoopedAge, B=1)
        )
        Period=0.4
        Scale=&FloatFromCurve(FloatCurve=DataInterface() {propertyValues {Curve {keys [<at 0 0 RCIM_Cubic>,<at 0.2717 1>,<at 1 5>]}}}, Scale Curve=4)
        Bias=0
       )
      )
      Destination Space=::Simulation
      Apply Scale=true
     )
    )
    Particles.PosNoodle2:NiagaraPosition=&AddVectorToPosition(
     Position=&AddVectorToPosition(
      Position=$Particles.Position
      Vector=&TransformVector(
       Vector=&MakeVector(
        X=5
        Y=&Sine(
         Normalized Angle=&Add_Float(
          A=&Multiply_Float(A=$Particles.NormalizedAge, B=-1.3)
          B=&Multiply_Float(A=$Emitter.LoopedAge, B=1)
         )
         Period=&MaskFloatBySpawnGroup(Masked Float=0.5, Passthrough Float=-0.5)
         Scale=&FloatFromCurve(
          FloatCurve=DataInterface() {propertyValues {Curve {keys [{arriveTangent=0.1945, interpMode=RCIM_Cubic, leaveTangent=0.1945, tangentMode=RCTM_User, time=0, value=0}, <at 0.2717 1>, <at 1 5>]}}}
          Scale Curve=5
         )
         Bias=0
        )
        Z=&Sine(
         Normalized Angle=&Add_Float(
          A=&Multiply_Float(A=$Particles.NormalizedAge, B=-1.3)
          B=&Multiply_Float(A=$Emitter.LoopedAge, B=1)
         )
         Period=0.6
         Scale=&FloatFromCurve(FloatCurve=DataInterface() {propertyValues {Curve {keys [<at 0 0 RCIM_Cubic>,<at 0.2717 1>,<at 1 5>]}}}, Scale Curve=5)
         Bias=0
        )
       )
       Destination Space=::Simulation
       Apply Scale=true
      )
     )
     Vector=<0 -2.5 -2.5>
    )
    Particles.PosNoodle3:NiagaraPosition=&AddVectorToPosition(
     Position=&AddVectorToPosition(
      Position=$Particles.Position
      Vector=&TransformVector(
       Vector=&MakeVector(
        X=10
        Y=&Sine(
         Normalized Angle=&Add_Float(
          A=&Multiply_Float(A=$Particles.NormalizedAge, B=-1.3)
          B=&Multiply_Float(A=$Emitter.LoopedAge, B=1)
         )
         Period=&MaskFloatBySpawnGroup(Masked Float=0.5, Passthrough Float=-0.5)
         Scale=&FloatFromCurve(
          FloatCurve=DataInterface() {propertyValues {Curve {keys [
              {arriveTangent=0.1945, interpMode=RCIM_Cubic, leaveTangent=0.1945, tangentMode=RCTM_User, time=0, value=0}
              {arriveTangent=0.5255, interpMode=RCIM_Cubic, leaveTangent=0.5255, tangentMode=RCTM_User, time=0.2717, value=1}
              <at 1 4>
             ]}}}
          Scale Curve=5
         )
         Bias=0
        )
        Z=&Sine(
         Normalized Angle=&Add_Float(
          A=&Multiply_Float(A=$Particles.NormalizedAge, B=-1.3)
          B=&Multiply_Float(A=$Emitter.LoopedAge, B=1)
         )
         Period=-0.5
         Scale=&FloatFromCurve(FloatCurve=DataInterface() {propertyValues {Curve {keys [<at 0 0 RCIM_Cubic>,<at 0.2717 1>,<at 1 4>]}}}, Scale Curve=5)
         Bias=0
        )
       )
       Destination Space=::Simulation
       Apply Scale=true
      )
     )
     Vector=<0 2.5 2.5>
    )
    Particles.PosNoodle4:NiagaraPosition=&AddVectorToPosition(
     Position=$Particles.Position
     Vector=&TransformVector(
      Vector=&Add_Vector(
       A=&MakeVector(
        X=-15
        Y=&Sine(
         Normalized Angle=&Add_Float(
          A=&Multiply_Float(A=$Particles.NormalizedAge, B=-1.3)
          B=&Multiply_Float(A=$Emitter.LoopedAge, B=1)
         )
         Period=&MaskFloatBySpawnGroup(Masked Float=0.6, Passthrough Float=-0.6)
         Scale=&FloatFromCurve(
          FloatCurve=DataInterface() {propertyValues {Curve {keys [{arriveTangent=0.1945, interpMode=RCIM_Cubic, leaveTangent=0.1945, tangentMode=RCTM_User, time=0, value=0}, <at 0.2717 1>, <at 1 5>]}}}
          Scale Curve=6
         )
         Bias=0
        )
        Z=&Sine(
         Normalized Angle=&Add_Float(
          A=&Multiply_Float(A=$Particles.NormalizedAge, B=-1.3)
          B=&Multiply_Float(A=$Emitter.LoopedAge, B=1)
         )
         Period=0.5
         Scale=&FloatFromCurve(FloatCurve=DataInterface() {propertyValues {Curve {keys [<at 0 0 RCIM_Cubic>,<at 0.2717 1>,<at 1 5>]}}}, Scale Curve=-5)
         Bias=0
        )
       )
       B=<0 2.5 -2.5>
      )
      Destination Space=::Simulation
      Apply Scale=true
     )
    )
   ) {isSetParameters=true}
   /SetVariables(Particles.Position=&AddVectorToPosition(
     Position=$Particles.Position
     Vector=/VehicleCosmetics/FX/Niagara/Modules/VC_ConvertVectorToWorld(Vector:Vector3f=&MakeVector(
       X=0
       Y=&FloatFromCurve(
        FloatCurve=DataInterface() {propertyValues {Curve {defaultValue=1, keys [{tangentMode=RCTM_User, time=0, value=0}, <at 0.5 0>, {tangentMode=RCTM_User, time=1, value=-25}]}}}
        Scale Curve=0.1
       )
       Z=&FloatFromCurve(
        FloatCurve=DataInterface() {propertyValues {Curve {keys [
            <at 0 0 RCIM_Cubic>
            {arriveTangent=5.5287, interpMode=RCIM_Cubic, leaveTangent=5.5287, tangentMode=RCTM_User, time=0.25, value=15}
            {arriveTangent=3.1163, interpMode=RCIM_Cubic, leaveTangent=3.1163, tangentMode=RCTM_User, time=0.5, value=16}
            {arriveTangent=-14.8689, interpMode=RCIM_Cubic, leaveTangent=-14.8689, tangentMode=RCTM_User, time=0.8, value=25}
            {arriveTangent=-3.7582, interpMode=RCIM_Cubic, leaveTangent=-3.7582, tangentMode=RCTM_User, time=1, value=0}
           ]}}}
        Scale Curve=0.5
       )
      ))
    )) {enabled=false, isSetParameters=true}
  ]
  rendererClasses [&NiagaraRibbonRendererProperties]
  renderers [
   {rendererClass=&NiagaraRibbonRendererProperties}
   {rendererClass=&NiagaraRibbonRendererProperties, rendererIndex=1}
   {rendererClass=&NiagaraRibbonRendererProperties, rendererIndex=2}
   {rendererClass=&NiagaraRibbonRendererProperties, rendererIndex=3}
  ]
 }
 NE_VC_DriftTrail_Sprite_Parent {
  emitterUpdateScript [
   &EmitterState()
   &SpawnRate(SpawnRate=15, Use Spawn Probability=false)
   &SpawnRate(SpawnGroup=1, SpawnRate=10, Use Spawn Probability=false) {module=SpawnRate001}
  ]
  particleSpawnScript [
   &VC_PaintedVariantsOverride() {enabled=false}
   /SetVariables(Particles.VisibilityTag=&SpawnGroup()) {isSetParameters=true}
   &InitializeParticle(
    Color Mode=::NewEnumerator0
    Lifetime=0.5
    Sprite Rotation Mode=::NewEnumerator2
    Sprite Size Mode=::NewEnumerator3
    Uniform Sprite Size=&SelectFloatFromArray(Direct Array Index=&SpawnGroup(), Float Selection Array=DataInterface() {propertyValues {FloatData [35,17]}})
    UsePositionOffset=false
    Write Material Random=true
   )
   &PartitionParticles()
   &SkeletalMeshLocation(
    Bone / Socket Index=&Add_Integer(A=$Particles.PartitionParticles.Partition, B=4)
    Bone / Socket Sampling Mode=::NewEnumerator5
    Interpolate Mesh Sampling=false
    MaskSpawnGroup=false
    Mesh Sampling Type=::NewEnumerator6
    Sampled Position Offset=&AddVectorToPosition(Position=<-90 0 -20>, Vector=$User.offset)
    Skeletal Mesh=$System.SkeletalMesh
   )
   &AddVelocity(Velocity=<-450 0 0>)
   &SubUVAnimation(
    "Sprite Renderer "=DataInterface() {propertyValues {SpriteRenderer=&self:NE_VC_DriftTrail_Sprite_Parent.NiagaraSpriteRendererProperties_0}}
    SubUV Animation Mode:ENiagara_SubUVLookupModeV2=::NewEnumerator2
    UseEndFrame=false
    UseStartFrame=false
   )
   &ShapeLocation(MaskSpawnGroup=false, Override Seed=false, Sphere Radius=20, UseOwnerScale=true)
  ]
  particleUpdateScript [
   &FN_DisableTireFX(
    Mask Index Attribute:NiagaraInt32=$Particles.PartitionParticles.Partition
    Mask Index FL=2
    Mask Index FR=3
    Mask Index RL=0
    Mask Index RR=1
   )
   &Drag(
    Drag=&FloatFromCurve(
     FloatCurve=DataInterface() {propertyValues {Curve {defaultValue=1, keys [{tangentMode=RCTM_User, time=0.3, value=0}, {tangentMode=RCTM_User, time=1, value=1}]}}}
     Scale Curve=2
    )
    Use Linear Drag=true
    Use Rotational Drag=true
   )
   &ParticleState()
   &ScaleSpriteSize(Uniform Curve Sprite Scale=DataInterface() {propertyValues {Curve {
       defaultValue=0
       keys [{interpMode=RCIM_Cubic, tangentMode=RCTM_User, time=0.3, value=1}, {interpMode=RCIM_Cubic, tangentMode=RCTM_User, time=1, value=0.5}]
      }}})
   &SpriteRotationRate(Rotation Rate=&RandomRangeFloat(
     Minimum=&SelectFloatFromArray(
      Array Sampling Mode=::NewEnumerator0
      Direct Array Index=?
      Float Selection Array=DataInterface() {propertyValues {FloatData [-500,-800]}}
      Random Seed=0
     )
     Maximum=&SelectFloatFromArray(
      Array Sampling Mode=::NewEnumerator0
      Direct Array Index=?
      Float Selection Array=DataInterface() {propertyValues {FloatData [500,850]}}
      Random Seed=0
     )
    ))
   &SolveForcesAndVelocity(Clamp Velocity=false, Limit Acceleration=false)
  ]
  rendererClasses [&NiagaraSpriteRendererProperties]
  renderers [
   {rendererClass=&NiagaraSpriteRendererProperties}
   {rendererClass=&NiagaraSpriteRendererProperties, rendererIndex=1}
  ]
 }
 NE_VC_DriftTrail_WheelWrap_Parent {
  emitterUpdateScript [
   &EmitterState(
    Inactive Response=::NewEnumerator1
    Life Cycle Mode=::NewEnumerator1
    Loop Behavior=::NewEnumerator1
    Loop Duration Mode=::NewEnumerator1
    UseLoopDelay=false
   )
   &SpawnBurst_Instantaneous(Spawn Count=2, Use Spawn Probability=false, UseLoopCountLimit=false)
  ]
  particleSpawnScript [
   &VC_PaintedVariantsOverride(
    Black=<rgba 0.015 0.015 0.015 10>
    BurntSienna=<rgba 0.2396 0.057 0.0311 1>
    Cobalt=<rgba 0.024 0.0436 1.2 1>
    Crimson=<rgba 0.4635 0 0.0126 5>
    DefaultColor=<rgba 0.9479 0.6517 0.237 1>
    ForestGreen=<rgba 0 0.276 0.0139 5>
    Grey=<rgba 0.2 0.2 0.2 5>
    Lime=<rgba 0.2599 0.4427 0 5>
    Orange=<rgba 0.6724 0.1426 0.0038 10>
    Pink=<rgba 0.7347 0.2348 0.7448 10>
    Purple=<rgba 0.23 0 0.6 10>
    Saffron=<rgba 0.8854 0.6788 0 5>
    SkyBlue=<rgba 0.0197 0.3277 0.6 5>
    TitaniumWhite=<rgba 0.5596 0.5596 0.5625 5>
   )
   &InitializeParticle(
    Color Mode=::NewEnumerator0
    Mesh Scale Mode=::NewEnumerator4
    Sprite Rotation Mode=::NewEnumerator2
    Sprite Size Mode=::NewEnumerator3
    UsePositionOffset=false
    Write Material Random=true
   )
   &PartitionParticles()
   &SkeletalMeshLocation(
    Bone / Socket Index=&Add_Integer(A=$Particles.PartitionParticles.Partition, B=4)
    Bone / Socket Sampling Mode=::NewEnumerator5
    Interpolate Mesh Sampling=false
    MaskSpawnGroup=false
    Mesh Sampling Type=::NewEnumerator6
    Skeletal Mesh=$System.SkeletalMesh
   )
  ]
  particleUpdateScript [
   &FN_DisableTireFX(
    Mask Index Attribute:NiagaraInt32=$Particles.PartitionParticles.Partition
    Mask Index FL=2
    Mask Index FR=3
    Mask Index RL=0
    Mask Index RR=1
   )
   &VC_WheelWrapScale(Wheel Index Attribute:NiagaraInt32=&Add_Integer(A=$Particles.PartitionParticles.Partition, B=2))
   &ParticleState(Kill Particles When Lifetime Has Elapsed=false, Loop Particles Lifetime=true)
   &ScaleColor(Scale Alpha=15, Scale RGB=&VectorFromFloat(Value=2), ScaleA=true, ScaleRGB=true)
   &UpdateMeshOrientation(Rotation Coordinate Space=::NewEnumerator2, Rotation Vector=<0 1 0>) {enabled=false}
   &SkeletalMeshLocation(
    Bone / Socket Index=&Add_Integer(A=$Particles.PartitionParticles.Partition, B=4)
    Bone / Socket Sampling Mode=::NewEnumerator5
    MaskSpawnGroup=false
    Mesh Sampling Type=::NewEnumerator6
    Skeletal Mesh=$System.SkeletalMesh
   ) {module=SkeletalMeshLocation001}
   &SolveForcesAndVelocity(Clamp Velocity=false, Limit Acceleration=false)
  ]
  rendererClasses [&NiagaraMeshRendererProperties]
  renderers [{rendererClass=&NiagaraMeshRendererProperties, rendererIndex=0}]
 }
 SPRT_Drops {
  emitterUpdateScript [
   &EmitterState()
   &SpawnRate(SpawnRate=30, Use Spawn Probability=false)
  ]
  particleSpawnScript [
   &VC_PaintedVariantsOverride(
    BurntSienna=<rgba 0.151 0.029 0.0073 2>
    Cobalt=<rgba 0 0.0124 0.6719 2>
    Crimson=<rgba 0.224 0.0105 0.0163 2>
    DefaultColor=<rgba 0.4635 0.3351 0.1135 7>
    ForestGreen=<rgba 0 0.151 0.0076 2>
    Grey=<rgba 0.0938 0.0938 0.0938 2>
    Lime=<rgba 0.0964 0.1927 0 2>
    Orange=<rgba 0.1302 0.0454 0 2>
    Pink=<rgba 0.5052 0.0605 0.3435 2>
    Purple=<rgba 0.1176 0 0.2865 2>
    Saffron=<rgba 0.2188 0.1677 0 2>
    SkyBlue=<rgba 0 0.1391 1 2>
    TitaniumWhite=<rgba 0.4905 0.491 0.5208 2>
   )
   &InitializeParticle(
    Color=$Particles.Color
    Lifetime=/VehicleCosmetics/FX/Niagara/Modules/MergeEmitters/VC_FloatBySpawnGroup(
     SpawnGroup or ExecIndex:NiagaraInt32=0
     CustomInt:NiagaraInt32=0
     MaximumSpawnGroups:NiagaraInt32=2
     Float SG 0:NiagaraFloat=&RandomRangeFloat(Minimum=0.4, Maximum=0.4)
     Float SG 1:NiagaraFloat=&RandomRangeFloat(Minimum=0.3, Maximum=0.55)
     Float SG 2:NiagaraFloat=0
     Float SG 3:NiagaraFloat=0
     Float SG 4:NiagaraFloat=0
     Float SG 5:NiagaraFloat=0
    )
    Sprite Rotation Mode=::NewEnumerator2
    Sprite Size=&RandomRangeVector2D(
     Minimum=&Multiply_Vector2D_ByFloat(Float=1, Vector2D=<5 5>)
     Maximum=&Multiply_Vector2D_ByFloat(Float=0.8, Vector2D=<7.8 7.8>)
    )
    Sprite Size Mode=::NewEnumerator4
    UsePositionOffset=false
    Write Material Random=true
   )
   &PartitionParticles()
   &SkeletalMeshLocation(
    Bone / Socket Index=&Add_Integer(A=$Particles.PartitionParticles.Partition, B=4)
    Bone / Socket Sampling Mode=::NewEnumerator5
    Interpolate Mesh Sampling=false
    MaskSpawnGroup=false
    Mesh Sampling Type=::NewEnumerator6
    Sampled Position Offset=&AddVectorToPosition(Position=<-100 0 -20>, Vector=$User.offset)
    Skeletal Mesh=$System.SkeletalMesh
   )
   &AddVelocity(Velocity=&RandomRangeVector(Minimum=<-70 35 25>, Maximum=<-50 -35 -12>))
   &ShapeLocation(MaskSpawnGroup=false, Override Seed=false, Sphere Radius=25, UseOwnerScale=true)
  ]
  particleUpdateScript [
   &FN_DisableTireFX(
    Mask Index Attribute:NiagaraInt32=$Particles.PartitionParticles.Partition
    Mask Index FL=2
    Mask Index FR=3
    Mask Index RL=0
    Mask Index RR=1
   )
   &ParticleState()
   &ScaleVelocity(Velocity Scale=&VC_VectorBySpawnGroup(
     SpawnGroup or ExecIndex:NiagaraInt32=0
     CustomInt:NiagaraInt32=0
     MaximumSpawnGroups:NiagaraInt32=2
     Vector Group 0:Vector3f=<0 0 0>
     Vector Group 1:Vector3f=&VectorFromCurve(
      VectorCurve=DataInterface() {propertyValues {
        XCurve {
         defaultValue=1
         keys [
          {interpMode=RCIM_Cubic, tangentMode=RCTM_Break, time=0, value=1}
          {arriveTangent=-2.8919, arriveTangentWeight=0.3333, interpMode=RCIM_Cubic, tangentMode=RCTM_Break, time=1, value=0}
         ]
        }
        YCurve {
         defaultValue=1
         keys [
          {interpMode=RCIM_Cubic, tangentMode=RCTM_Break, time=0, value=1}
          {arriveTangent=-2.8919, arriveTangentWeight=0.3333, interpMode=RCIM_Cubic, tangentMode=RCTM_Break, time=1, value=0}
         ]
        }
        ZCurve {keys [
          {interpMode=RCIM_Cubic, tangentMode=RCTM_Break, time=0, value=1}
          {arriveTangent=-2.8919, interpMode=RCIM_Cubic, tangentMode=RCTM_Break, time=1, value=0}
         ]}
       }}
      Scale Curve=<5 5 5>
     )
     Vector Group 2:Vector3f=<0 0 0>
     Vector Group 3:Vector3f=<0 0 0>
     Vector Group 4:Vector3f=<0 0 0>
     Vector Group 5:Vector3f=<0 0 0>
    ))
   &ScaleSpriteSize(
    Uniform Curve Scale=2.5
    Uniform Curve Sprite Scale=DataInterface() {propertyValues {Curve {keys [
        {interpMode=RCIM_Cubic, tangentMode=RCTM_Break, time=0.2, value=1}
        {arriveTangent=-2.1991, interpMode=RCIM_Cubic, tangentMode=RCTM_Break, time=1, value=0.2}
       ]}}}
   )
   &DynamicMaterialParameters(
    Index 0 Param 1=&FloatFromCurve(FloatCurve=DataInterface() {propertyValues {Curve {keys [<at 0 1>,<at 1 1.5>]}}})
    Index 0 Param 2=&FloatFromCurve(FloatCurve=DataInterface() {propertyValues {Curve {keys [<at 0.45 -1>,<at 1 0>]}}})
    Index 0 Param 3=0.192
    Index 1 Param 1=1
    Index 1 Param 3=1
    Index 1 Param 4=2
    Module.Index0Float0Write=true
    Module.Index0Float1Write=true
    Module.Index0Float2Write=true
    Module.Index0Float3Write=true
    Module.Index1Float0Write=true
    Module.Index1Float1Write=true
    Module.Index1Float2Write=true
    Module.Index1Float3Write=true
    Write Parameter Index 1=::NewEnumerator1
   )
   &ScaleVelocity(Velocity Scale=&VectorFromCurve(VectorCurve=DataInterface() {propertyValues {XCurve {keys [<at 0 4>]}, YCurve {keys [<at 0 4>]}, ZCurve {keys [<at 0 4>,<at 0.5 4>,<at 1 -10>]}}})) {module=ScaleVelocity001}
   &SolveForcesAndVelocity(Clamp Velocity=false, Limit Acceleration=false)
  ]
  rendererClasses [&NiagaraSpriteRendererProperties]
  renderers [{rendererClass=&NiagaraSpriteRendererProperties}]
 }
}
summary {
 systemName=NS_DriftTrail_Ramen
 emitters [
  {emitterName=SPRT_Drops, rendererClasses [&NiagaraSpriteRendererProperties]}
  {
   emitterName=MSH_Cone_Soup
   simTarget=GPUComputeSim
   rendererClasses [&NiagaraMeshRendererProperties, &NiagaraSpriteRendererProperties]
  }
  {emitterName=NE_VC_DriftTrail_Sprite_Parent, rendererClasses [&NiagaraSpriteRendererProperties]}
  {emitterName=NE_VC_DriftTrail_WheelWrap_Parent, rendererClasses [&NiagaraMeshRendererProperties]}
  {emitterName=NE_VC_DriftTrail_Ribbon_Parent001, rendererClasses [&NiagaraRibbonRendererProperties]}
 ]
 userVariables [
  User.bIsFrontendPreview {type=NiagaraBool}
  User.bIsPainted {type=NiagaraBool}
  User.ColorVariantID {type=NiagaraInt32}
  User.offset {type=Vector3f, description="offset from wheels(set through cosmetic template)", defaultValue {struct=Vector3f, value=<0 0 0>}}
  User.PaintedColor {type=LinearColor, description="", defaultValue {struct=LinearColor, value=<rgba 1 1 1 1>}}
  User.PaintedScalar {type=NiagaraFloat, defaultValue=1, description=""}
  User.VehicleFlags {type=NiagaraInt32, defaultValue=806289408}
  User.Wheel_Radius_Array {
   type=NiagaraDataInterfaceArrayFloat
   defaultValue {
    struct=NiagaraExt_VariableValue_DataInterface
    value {
     dataInterface=&self:NiagaraDataInterfaceArrayFloat_1
     dataInterfaceClass=&NiagaraDataInterfaceArrayFloat
    }
   }
  }
  User.Wheel_Width_Array {
   type=NiagaraDataInterfaceArrayFloat
   defaultValue {
    struct=NiagaraExt_VariableValue_DataInterface
    value {
     dataInterface=&self:NiagaraDataInterfaceArrayFloat_0
     dataInterfaceClass=&NiagaraDataInterfaceArrayFloat
    }
   }
  }
  User.WheelTouchingArray {
   type=NiagaraDataInterfaceArrayFloat
   defaultValue {
    struct=NiagaraExt_VariableValue_DataInterface
    value {
     dataInterface=&self:NiagaraDataInterfaceArrayFloat_2
     dataInterfaceClass=&NiagaraDataInterfaceArrayFloat
    }
   }
  }
 ]
}
